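# Physics asset for the DistantEchoPro medium male player body's saber accessories, covering a saber hanging on the left of the belt and a tube on the right.
# Two bone chains run from the pelvis through belt and clasp to saber or tube, with tight twist and moderate swing limits below the belt.
# Capsules on pelvis, lower spine and thighs approximate the body.
# Their constraints and both belt-to-pelvis joints leave angular motion free.
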PhysicsAsset /Game/Characters/Player/Male/Medium/Bodies/M_MED_DistantEchoPro/Meshes/M_MED_DistantEchoPro_Saber_Physic
bodies {
 dyn_belt_l [{
   shapeType=Capsule
   length=1.4191
   radius=1.1888
   shapeName=dyn_belt_l_capsule
   center=<4.1921 0.2401 0.0167>
   rotation=<rot 0 94.6775 -64.2343>
  }]
 dyn_belt_r [{
   shapeType=Capsule
   length=1.5432
   radius=1.1072
   shapeName=dyn_belt_r_capsule
   center=<1.6502 0.1351 -0.344>
   rotation=<rot 0 104.1353 -77.322>
  }]
 dyn_clasp_l [{shapeType=Capsule, length=1.5826, radius=1.3394, shapeName=dyn_clasp_l_capsule, center=<0 1.5 0>}]
 dyn_clasp_r [{shapeType=Capsule, length=2.6929, radius=0.921, shapeName=dyn_clasp_r_capsule, rotation=<rot -0.0001 0 53.4376>}]
 dyn_saber_l [{shapeType=Capsule, length=20.5912, radius=2.9238, shapeName=dyn_saber_l_capsule, center=<13.3601 -2 0>, rotation=<rot 0 75.5542 -87.8327>}]
 dyn_tube_r [{
   shapeType=Capsule
   length=7.1347
   radius=3.4606
   shapeName=dyn_tube_r_capsule
   center=<2.5061 -0.3138 0.4231>
   rotation=<rot -0.0001 0 53.4376>
  }]
 pelvis [{shapeType=Capsule, length=11.3005, radius=9.74, center=<2.3893 0 0>}]
 spine_01 [{shapeType=Capsule, length=11.3416, radius=13.6584, shapeName=spine_01_capsule, center=<5.5289 0.5405 1.5888>, rotation=<rot 90 0 0>}]
 thigh_l [{shapeType=Capsule, length=51.7376, radius=9.3107, shapeName=thigh_l_capsule, center=<-25.9919 0 0>, rotation=<rot 90 0 0>}]
 thigh_r [{shapeType=Capsule, length=44.2976, radius=9.5229, shapeName=thigh_r_capsule, center=<22.6889 0 0>, rotation=<rot 90 0 0>}]
}
constraints [
 {bone1Name=dyn_clasp_r, bone2Name=dyn_belt_r, swing1LimitDegrees=25, swing2LimitDegrees=30, twistLimitDegrees=1}
 {bone1Name=dyn_tube_r, bone2Name=dyn_clasp_r, swing1LimitDegrees=25, swing2LimitDegrees=30, twistLimitDegrees=1}
 {
  bone1Name=thigh_l
  bone2Name=pelvis
  swing1LimitDegrees=45
  swing1Motion=Free
  swing2LimitDegrees=45
  swing2Motion=Free
  twistLimitDegrees=45
  twistMotion=Free
 }
 {
  bone1Name=thigh_r
  bone2Name=pelvis
  swing1LimitDegrees=45
  swing1Motion=Free
  swing2LimitDegrees=45
  swing2Motion=Free
  twistLimitDegrees=45
  twistMotion=Free
 }
 {bone1Name=dyn_clasp_l, bone2Name=dyn_belt_l, swing1LimitDegrees=10, swing2LimitDegrees=20, twistLimitDegrees=1}
 {bone1Name=dyn_saber_l, bone2Name=dyn_clasp_l, swing1LimitDegrees=15, swing2LimitDegrees=30, twistLimitDegrees=1}
 {
  bone1Name=dyn_belt_l
  bone2Name=pelvis
  swing1LimitDegrees=45
  swing1Motion=Free
  swing2LimitDegrees=45
  swing2Motion=Free
  twistLimitDegrees=45
  twistMotion=Free
 }
 {
  bone1Name=spine_01
  bone2Name=pelvis
  swing1LimitDegrees=45
  swing1Motion=Free
  swing2LimitDegrees=45
  swing2Motion=Free
  twistLimitDegrees=45
  twistMotion=Free
 }
 {
  bone1Name=dyn_belt_r
  bone2Name=pelvis
  swing1LimitDegrees=45
  swing1Motion=Free
  swing2LimitDegrees=45
  swing2Motion=Free
  twistLimitDegrees=45
  twistMotion=Free
 }
]
