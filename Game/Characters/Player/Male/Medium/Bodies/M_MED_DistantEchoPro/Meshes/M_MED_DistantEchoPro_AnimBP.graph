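# Cosmetic AnimBP for the DistantEchoPro male body, adding secondary motion to robe, sleeves, belt and saber.
# The copied parent pose runs through deform, crouch and jump Control Rigs, driven by EventGraph movement state, then AnimDynamics and thigh constraints.
# A cached pose feeds separate front-end and in-game RigidBody chains, the latter using computed gravity, before bone offsets and constraints finish.
# Pouch AnimDynamics nodes are disconnected from the output.
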AnimBlueprint /Game/Characters/Player/Male/Medium/Bodies/M_MED_DistantEchoPro/Meshes/M_MED_DistantEchoPro_AnimBP
target_skeleton=&/M_MED_DistantEchoPro_Skeleton
graphs [
 AnimGraph {nodes [
   n1 = Anim Dynamics - Bone: dyn_pouch_bk(Alpha=1, ExternalForce="0,0,0") {
    angularDampingOverride=0.9999
    angularSpringConstant=500
    bAngularSpring=true
    bGravityOverrideInSimSpace=true
    bOverrideAngularDamping=true
    bUseGravityOverride=true
    simulationSpace=RootRelative
    alphaBoolBlend {blendOption=Linear}
    alphaScaleBiasClamp {interpSpeedIncreasing=20}
    boundBone {boneName=dyn_pouch_bk}
    componentAppliedLinearAccClamp=<1000 1000 1000>
    componentLinearAccScale=<0.2 0.1 0.1>
    componentLinearVelScale=<0.2 0.1 0.1>
    gravityOverride {z=-980}
    physicsBodyDefinitions [{
      sphereCollisionRadius=0
      boundBone {boneName=dyn_pouch_bk}
      boxExtents=<5 15 15>
      constraintSetup {angularLimitsMax=<1 3 2>, angularLimitsMin=<-1 -3 -2>, angularTarget=<0 0 0>}
      localJointOffset=<3 0 0>
     }]
    relativeSpaceBone {boneName=thigh_r}
   }
   Anim Dynamics - Bone: dyn_pouch_bk_r(ComponentPose=@n1, Alpha=1, ExternalForce="0,0,0") {
    angularDampingOverride=0.9999
    angularSpringConstant=500
    bAngularSpring=true
    bGravityOverrideInSimSpace=true
    bOverrideAngularDamping=true
    bUseGravityOverride=true
    simulationSpace=RootRelative
    alphaBoolBlend {blendOption=Linear}
    boundBone {boneName=dyn_pouch_bk_r}
    componentAppliedLinearAccClamp=<1000 1000 1000>
    componentLinearAccScale=<0.2 0.1 0.1>
    componentLinearVelScale=<0.2 0.1 0.1>
    gravityOverride {z=-980}
    physicsBodyDefinitions [{
      sphereCollisionRadius=0
      boundBone {boneName=dyn_pouch_bk_r}
      boxExtents=<5 15 15>
      constraintSetup {angularLimitsMax=<1 3 2>, angularLimitsMin=<-1 -3 -2>, angularTarget=<0 0 0>}
      localJointOffset=<3 0 0>
     }]
    relativeSpaceBone {boneName=thigh_r}
   }
   Anim Dynamics - Bone: dyn_pouch_fr(Alpha=1, ExternalForce="0,0,0") {
    angularDampingOverride=0.9999
    angularSpringConstant=500
    bAngularSpring=true
    bGravityOverrideInSimSpace=true
    bOverrideAngularDamping=true
    bUseGravityOverride=true
    simulationSpace=RootRelative
    alphaBoolBlend {blendOption=Linear}
    alphaScaleBiasClamp {interpSpeedIncreasing=20}
    boundBone {boneName=dyn_pouch_fr}
    componentAppliedLinearAccClamp=<1000 1000 1000>
    componentLinearAccScale=<0.2 0.1 0.1>
    componentLinearVelScale=<0.2 0.1 0.1>
    gravityOverride {z=-980}
    physicsBodyDefinitions [{
      sphereCollisionRadius=0
      boundBone {boneName=dyn_pouch_fr}
      boxExtents=<5 15 15>
      constraintSetup {angularLimitsMax=<1 3 2>, angularLimitsMin=<-1 -3 -2>, angularTarget=<0 0 0>}
      localJointOffset=<3 0 0>
     }]
    relativeSpaceBone {boneName=thigh_r}
   }
   n2 = "Transform (Modify) Bone - Bone: dyn_robe_bk_r"!(ComponentPose=Local To Component(LocalPose=Use cached pose 'RBAN'()), Alpha=1, Translation="0,-7,0") -> (Pose) {
    rotationMode=BMM_Replace
    rotationSpace=BCS_BoneSpace
    translationMode=BMM_Additive
    translationSpace=BCS_BoneSpace
    alphaBoolBlend {blendOption=Linear}
    alphaScaleBiasClamp {bInterpResult=true, interpSpeedDecreasing=5, interpSpeedIncreasing=7}
    boneToModify {boneName=dyn_robe_bk_r}
    translation {y=-7}
   }
   n3 = "Constraint - dyn_robe_fr (P)"(ComponentPose=@n2, Alpha=0.8, ConstraintWeights_0=1) {
    alphaBoolBlend {blendOption=Linear}
    alphaScaleBiasClamp {bClampResult=true, bInterpResult=true, clampMax=0.7, interpSpeedDecreasing=5, interpSpeedIncreasing=7}
    boneToModify {boneName=dyn_robe_fr}
    constraintSetup [{targetBone {boneName=thigh_l}}]
    constraintWeights [1]
   }
   n4 = "Transform (Modify) Bone - Bone: dyn_robe_fr"!(ComponentPose=@n3, Alpha=1, Translation="0,-11,0", Rotation="0,0,5") -> (Pose) {
    rotationMode=BMM_Replace
    rotationSpace=BCS_BoneSpace
    translationMode=BMM_Additive
    translationSpace=BCS_BoneSpace
    alphaBoolBlend {blendOption=Linear}
    alphaScaleBiasClamp {bInterpResult=true, interpSpeedDecreasing=5, interpSpeedIncreasing=7}
    boneToModify {boneName=dyn_robe_fr}
    rotation {roll=5}
    translation {y=-11}
   }
   n5 = "Constraint - dyn_robe_bk_l (P)"(ComponentPose=@n4, Alpha=0.8, ConstraintWeights_0=1) {
    alphaBoolBlend {blendOption=Linear}
    alphaScaleBiasClamp {bClampResult=true, bInterpResult=true, clampMax=0.7, interpSpeedDecreasing=5, interpSpeedIncreasing=7}
    boneToModify {boneName=dyn_robe_bk_l}
    constraintSetup [{targetBone {boneName=thigh_l}}]
    constraintWeights [1]
   }
   n6 = "Transform (Modify) Bone - Bone: dyn_robe_bk"!(ComponentPose=@n5, Alpha=Get zipline() -> (zipline:double), Translation="-1.114,-8.858,-0.466") -> (Pose) {
    rotationMode=BMM_Replace
    rotationSpace=BCS_BoneSpace
    translationMode=BMM_Additive
    translationSpace=BCS_BoneSpace
    alphaBoolBlend {blendOption=Linear}
    alphaScaleBiasClamp {bInterpResult=true, interpSpeedDecreasing=5, interpSpeedIncreasing=7}
    boneToModify {boneName=dyn_robe_bk}
    rotation=<rot 0.0228 -30.9374 -0.0357>
    translation=<-1.114 -8.858 -0.466>
   }
   n7 = "Transform (Modify) Bone - Bone: dyn_robe_bk"(ComponentPose=@n6, Alpha=1, Translation="-1.114,-8.858,-0.466") {
    rotationMode=BMM_Replace
    rotationSpace=BCS_BoneSpace
    translationMode=BMM_Additive
    translationSpace=BCS_BoneSpace
    alphaBoolBlend {blendOption=Linear}
    alphaScaleBiasClamp {bInterpResult=true, interpSpeedDecreasing=5, interpSpeedIncreasing=7}
    boneToModify {boneName=dyn_robe_bk}
    rotation=<rot 0.0228 -30.9374 -0.0357>
    translation=<-1.114 -8.858 -0.466>
   }
   n8 = "Transform (Modify) Bone - Bone: dyn_robe_bk"(ComponentPose=@n7, Alpha=1, Translation="3.263,-5.442,-0.001") {
    rotationMode=BMM_Replace
    rotationSpace=BCS_BoneSpace
    translationMode=BMM_Additive
    translationSpace=BCS_BoneSpace
    alphaBoolBlend {blendOption=Linear}
    alphaScaleBiasClamp {bInterpResult=true, interpSpeedDecreasing=5, interpSpeedIncreasing=7}
    boneToModify {boneName=dyn_robe_bk}
    rotation=<rot 0.0003 22.5001 0.0021>
    translation=<3.2631 -5.4421 -0.001>
   }
   n9 = "Transform (Modify) Bone - Bone: dyn_saber_l"!(ComponentPose=@n8, Alpha=1, Rotation="-62.803,50.6948,-20") -> (Pose) {
    rotationMode=BMM_Replace
    rotationSpace=BCS_BoneSpace
    translationMode=BMM_Additive
    translationSpace=BCS_BoneSpace
    alphaBoolBlend {blendOption=Linear}
    alphaScaleBiasClamp {bInterpResult=true, interpSpeedDecreasing=5, interpSpeedIncreasing=7}
    boneToModify {boneName=dyn_saber_l}
    rotation=<rot -62.803 50.6948 -20>
   }
   n10 = "Transform (Modify) Bone - Bone: dyn_belt_r"!(ComponentPose=@n9, Alpha=1, Translation="-0.752,-4.274,0.826", Rotation="-50.469,112.629,-81.881") -> (Pose) {
    rotationMode=BMM_Replace
    rotationSpace=BCS_BoneSpace
    translationMode=BMM_Replace
    translationSpace=BCS_BoneSpace
    alphaBoolBlend {blendOption=Linear}
    alphaScaleBiasClamp {bInterpResult=true, interpSpeedDecreasing=5, interpSpeedIncreasing=7}
    boneToModify {boneName=dyn_belt_r}
    rotation=<rot -50.469 112.629 -81.881>
    translation=<-0.752 -4.274 0.826>
   }
   n11 = "Transform (Modify) Bone - Bone: dyn_saber_l"!(ComponentPose=@n10, Alpha=1, Translation:Vector="0,0,0", Rotation="-40.970,38.442,-27.468") -> (Pose) {
    rotationMode=BMM_Replace
    rotationSpace=BCS_BoneSpace
    translationMode=BMM_Replace
    translationSpace=BCS_BoneSpace
    alphaBoolBlend {blendOption=Linear}
    alphaScaleBiasClamp {bInterpResult=true, interpSpeedDecreasing=5, interpSpeedIncreasing=7}
    boneToModify {boneName=dyn_saber_l}
    rotation=<rot -40.97 38.4421 -27.4681>
   }
   n12 = "Transform (Modify) Bone - Bone: dyn_tube_r"!(ComponentPose=@n11, Alpha=1, Translation="0,0,0", Rotation="-22.136,-30.565,12.592") -> (Pose) {
    rotationMode=BMM_Replace
    rotationSpace=BCS_BoneSpace
    translationMode=BMM_Replace
    translationSpace=BCS_BoneSpace
    alphaBoolBlend {blendOption=Linear}
    alphaScaleBiasClamp {bInterpResult=true, interpSpeedDecreasing=5, interpSpeedIncreasing=7}
    boneToModify {boneName=dyn_tube_r}
    rotation=<rot -22.136 -30.565 12.592>
   }
   n13 = "Transform (Modify) Bone - Bone: attach_backpack"!(ComponentPose=@n12, Alpha=1, Translation="-0.457,2.128,-0.037", Rotation="0.001,-0.001,5.624") -> (Pose) {
    rotationMode=BMM_Replace
    rotationSpace=BCS_BoneSpace
    translationMode=BMM_Replace
    translationSpace=BCS_BoneSpace
    alphaBoolBlend {blendOption=Linear}
    alphaScaleBiasClamp {bInterpResult=true, interpSpeedDecreasing=5, interpSpeedIncreasing=7}
    boneToModify {boneName=attach_backpack}
    rotation=<rot 0.0011 -0.001 5.624>
    translation=<-0.457 2.128 -0.037>
   }
   n14 = "Transform (Modify) Bone - Bone: dyn_robe_fr"!(ComponentPose=@n13, Alpha=1, Translation="-0.000,-5,-0.000") -> (Pose) {
    rotationMode=BMM_Replace
    rotationSpace=BCS_BoneSpace
    translationMode=BMM_Additive
    translationSpace=BCS_BoneSpace
    alphaBoolBlend {blendOption=Linear}
    alphaScaleBiasClamp {bInterpResult=true, interpSpeedDecreasing=5, interpSpeedIncreasing=7}
    boneToModify {boneName=dyn_robe_fr}
    translation {y=-5}
   }
   Output Pose(Result=Component To Local(ComponentPose=@n14)) {layerGroup=None}
   n15 = Control Rig(Source=Copy Pose From Mesh() {bUseAttachedParent=true}) {
    alphaBoolBlend {blendOption=Linear}
    controlRigAssetReference {blueprintRigClass=&/M_MED_DistantEchoPro_Deform_Rig2.M_MED_DistantEchoPro_Deform_Rig2_C}
   }
   n16 = Control Rig!(
    Source=@n15
    Alpha=Get Crouching() -> (Crouching:double)
    Backward:double=Get Backward() -> (Backward:double)
    Forward:double=Get Forward() -> (Forward:double)
    Left:double=Get Left() -> (Left:double)
    Right:double=Get Right() -> (Right:double)
   ) -> (Pose) {
    alphaBoolBlend {blendOption=Linear}
    controlRigAssetReference {blueprintRigClass=&/M_MED_DistantEchoPro_Crouch_CtrlRig.M_MED_DistantEchoPro_Crouch_CtrlRig_C}
   }
   n17 = Control Rig!(
    Source=@n16
    Alpha=Get Jumping() -> (Jumping:double)
    Jump Forward:bool=Get Jump_Forward() -> (Jump_Forward:bool)
    Tactical:double=Get Tactical_Jump() -> (Tactical_Jump:double)
   ) -> (Pose) {
    alphaBoolBlend {blendOption=Linear}
    controlRigAssetReference {blueprintRigClass=&/M_MED_DistantEchoPro_Deform_CtrlRig.M_MED_DistantEchoPro_Deform_CtrlRig_C}
   }
   n18 = Anim Dynamics - Bone: dyn_sleeve_l(ComponentPose=Local To Component(LocalPose=@n17), Alpha=0.7, ExternalForce="0,0,0") {
    angularDampingOverride=0.98
    angularSpringConstant=75
    bAngularSpring=true
    bGravityOverrideInSimSpace=true
    bOverrideAngularDamping=true
    bUseSphericalLimits=true
    simulationSpace=RootRelative
    alphaBoolBlend {blendOption=Linear}
    boundBone {boneName=dyn_sleeve_l}
    componentAppliedLinearAccClamp=<1500 1500 1500>
    componentLinearAccScale=<0.5 0.5 0.2>
    componentLinearVelScale=<0.5 0.5 0.2>
    physicsBodyDefinitions [{
      collisionType=CustomSphere
      sphereCollisionRadius=4
      boundBone {boneName=dyn_sleeve_l}
      boxExtents=<25 15 15>
      constraintSetup {angularLimitsMax=<1 15 20>, angularLimitsMin=<-1 -5 -20>, angularTarget=<1 0 0>}
      localJointOffset=<5 0 0>
     }]
    relativeSpaceBone {boneName=pelvis}
    sphericalLimits [{limitRadius=6, drivingBone {boneName=upperarm_twist_02_r}, sphereLocalOffset=<0 0 3>}]
   }
   n19 = Anim Dynamics - Bone: dyn_sleeve_r(ComponentPose=@n18, Alpha=0.7, ExternalForce="0,0,0") {
    angularDampingOverride=0.98
    angularSpringConstant=75
    bAngularSpring=true
    bGravityOverrideInSimSpace=true
    bOverrideAngularDamping=true
    bUseSphericalLimits=true
    simulationSpace=RootRelative
    alphaBoolBlend {blendOption=Linear}
    boundBone {boneName=dyn_sleeve_r}
    componentAppliedLinearAccClamp=<1500 1500 1500>
    componentLinearAccScale=<0.5 0.5 0.2>
    componentLinearVelScale=<0.5 0.5 0.2>
    physicsBodyDefinitions [{
      collisionType=CustomSphere
      sphereCollisionRadius=4
      boundBone {boneName=dyn_sleeve_r}
      boxExtents=<25 15 15>
      constraintSetup {angularLimitsMax=<1 15 20>, angularLimitsMin=<-1 -5 -20>, angularTarget=<1 0 0>}
      localJointOffset=<5 0 0>
     }]
    relativeSpaceBone {boneName=pelvis}
    sphericalLimits [{limitRadius=6, drivingBone {boneName=upperarm_twist_02_r}, sphereLocalOffset=<0 0 3>}]
   }
   n20 = Anim Dynamics - Bone: dyn_clasp_r(ComponentPose=@n19, Alpha=1, ExternalForce="0,0,0") {
    angularDampingOverride=0.98
    angularSpringConstant=75
    bAngularSpring=true
    bGravityOverrideInSimSpace=true
    bOverrideAngularDamping=true
    bUseGravityOverride=true
    bUseSphericalLimits=true
    simulationSpace=RootRelative
    alphaBoolBlend {blendOption=Linear}
    boundBone {boneName=dyn_clasp_r}
    componentAppliedLinearAccClamp=<3500 3500 3500>
    componentLinearAccScale=<0.9 0.9 0.9>
    componentLinearVelScale=<0.9 0.9 0.9>
    gravityOverride {z=-980}
    physicsBodyDefinitions [{
      collisionType=CustomSphere
      sphereCollisionRadius=4
      boundBone {boneName=dyn_clasp_r}
      boxExtents=<25 15 15>
      constraintSetup {angularLimitsMax=<1 35 10>, angularLimitsMin=<-1 -10 -20>, angularTarget=<1 0 0>}
      localJointOffset=<2 0 0>
     }]
    relativeSpaceBone {boneName=pelvis}
    sphericalLimits [{limitRadius=6, drivingBone {boneName=thigh_r}, sphereLocalOffset=<0 0 0>}]
   }
   n21 = Anim Dynamics - Bone: dyn_tube_r(ComponentPose=@n20, Alpha=1, ExternalForce="0,0,0") {
    angularDampingOverride=0.98
    angularSpringConstant=75
    bAngularSpring=true
    bGravityOverrideInSimSpace=true
    bOverrideAngularDamping=true
    bUseGravityOverride=true
    bUseSphericalLimits=true
    simulationSpace=RootRelative
    alphaBoolBlend {blendOption=Linear}
    boundBone {boneName=dyn_tube_r}
    componentAppliedLinearAccClamp=<3500 3500 3500>
    componentLinearAccScale=<0.9 0.9 0.9>
    componentLinearVelScale=<0.9 0.9 0.9>
    gravityOverride {z=-980}
    physicsBodyDefinitions [{
      collisionType=CustomSphere
      sphereCollisionRadius=4
      boundBone {boneName=dyn_tube_r}
      boxExtents=<25 15 15>
      constraintSetup {angularLimitsMax=<1 35 10>, angularLimitsMin=<-1 -10 -20>, angularTarget=<1 0 0>}
      localJointOffset=<2 0 0>
     }]
    relativeSpaceBone {boneName=pelvis}
    sphericalLimits [{limitRadius=6, drivingBone {boneName=thigh_r}, sphereLocalOffset=<0 0 0>}]
   }
   n22 = "Constraint - dyn_robe_fr_r (P)"(ComponentPose=@n21, Alpha=0.8, ConstraintWeights_0=1) {
    alphaBoolBlend {blendOption=Linear}
    boneToModify {boneName=dyn_robe_fr_r}
    constraintSetup [{targetBone {boneName=thigh_r}}]
    constraintWeights [1]
   }
   n23 = "Constraint - dyn_robe_bk_r (P)"(ComponentPose=@n22, Alpha=0.8, ConstraintWeights_0=1) {
    alphaBoolBlend {blendOption=Linear}
    boneToModify {boneName=dyn_robe_bk_r}
    constraintSetup [{targetBone {boneName=thigh_r}}]
    constraintWeights [1]
   }
   Save cached pose 'Start'(Pose=Component To Local(ComponentPose=@n23)) {cacheName=Start}
   n24 = RigidBody!(ComponentPose=Local To Component(LocalPose=Use cached pose 'Start'()), Alpha, OverrideWorldGravity:Vector="0,0,-980", ExternalForce) -> (Pose) {
    bOverrideWorldGravity=true
    overridePhysicsAsset=&/M_MED_DistantEchoPro_Physics
    simulationSpace=BaseBoneSpace
    alphaBoolBlend {blendOption=Linear}
    baseBoneRef {boneName=pelvis}
    componentAppliedLinearAccClamp=<3000 3000 3000>
    componentLinearAccScale=<0.1 0.1 0.3>
    componentLinearVelScale=<0.1 0.1 0.3>
    overrideWorldGravity {z=-980}
   }
   n25 = RigidBody(ComponentPose=@n24, OverrideWorldGravity="0,0,-980") {
    bClampLinearTranslationLimitToRefPose=true
    bOverrideWorldGravity=true
    overridePhysicsAsset=&/M_MED_DistantEchoPro_SaberFE_Physic
    simulationSpace=BaseBoneSpace
    alphaBoolBlend {blendOption=Linear}
    baseBoneRef {boneName=spine_02}
    componentAppliedLinearAccClamp=<1000 1000 1000>
    componentLinearAccScale=<0.1 0.05 0.05>
    componentLinearVelScale=<0.1 0.05 0.05>
    overrideWorldGravity {z=-980}
   }
   n26 = Look At - Bone: dyn_belt_l(ComponentPose=@n25, Alpha=0.5) {
    alphaBoolBlend {blendOption=Linear}
    boneToModify {boneName=dyn_belt_l}
    lookAtLocation=<-49.032 -2.6263 -9.8851>
    lookAtTarget {boneReference {boneName=thigh_l}}
    lookAt_Axis {axis=<1 0>}
   }
   n27 = Get Gravity Override()
   n28 = RigidBody(ComponentPose=Local To Component(LocalPose=Use cached pose 'Start'()), OverrideWorldGravity=@n27) {
    bOverrideWorldGravity=true
    overridePhysicsAsset=&M_MED_DistantEchoPro_Physics
    simulationSpace=BaseBoneSpace
    alphaBoolBlend {blendOption=Linear}
    baseBoneRef {boneName=pelvis}
    componentAppliedLinearAccClamp=<3000 3000 3000>
    componentLinearAccScale=<0.1 0.1 0.3>
    componentLinearVelScale=<0.1 0.1 0.3>
    overrideWorldGravity {z=-980}
   }
   n29 = RigidBody(ComponentPose=@n28, OverrideWorldGravity=@n27) {
    bOverrideWorldGravity=true
    overridePhysicsAsset=&/M_MED_DistantEchoPro_Saber_Physic
    simulationSpace=BaseBoneSpace
    alphaBoolBlend {blendOption=Linear}
    baseBoneRef {boneName=pelvis}
    componentAppliedLinearAccClamp=<4500 4500 4500>
    componentLinearAccScale=<0.2 0.2 0.3>
    componentLinearVelScale=<0.2 0.2 0.3>
    overrideWorldGravity {z=-980}
   }
   Save cached pose 'RBAN'(Pose=Blend Poses by bool(BlendPose_0=Component To Local(ComponentPose=@n26), BlendPose_1=Component To Local(ComponentPose=@n29), bActiveValue=Get bIsInFrontEnd()) {blendType=Linear}) {cacheName=RBAN}
  ]}
 EventGraph {nodes [
   n1 = Event BlueprintUpdateAnimation()
   > Set Delta Time X(Delta Time X:double=@n1.DeltaTimeX) -> (Output_Get:double)
   > n2 = Sequence!(execute) -> (then_0, then_1, then_2:Exec, then_3:Exec, then_4:Exec)
   > n3 = Gravity Override State Function DEPRECATED(
    CharacterPartAnimInstance=Self-Reference()
    GameTimeInSeconds=Get Delta Time X() -> (Delta Time X:double)
    GravityOverrideParams=Get Gravity Override Params()
   )
   > Set Gravity Override(Gravity Override=@n3)
   Set Crouching(execute=@n2.then_1, Crouching:double=SelectFloat(A=1, B=0, bPickA=Get bIsCrouching())) -> (Output_Get:double)
   n4 = OR Boolean!(A=Get bIsJumping(), B=Get bIsFalling(), C:bool) -> (ReturnValue)
   n5 = Get bIsTacticalSprinting()
   > Set Tactical_Jump(Tactical_Jump:double=SelectFloat(A=1, B=0, bPickA=AND Boolean(A=@n4, B=@n5))) -> (Output_Get:double)
   > Set Crouch_move(Crouch_move:bool=AND Boolean(A=Get bIsCrouching(), B=Get bIsMoving2D())) -> (Output_Get:bool)
   n6 = NOT Boolean(A=Get bIsSkyDiving())
   n7 = OR Boolean(A=Get bIsJumping(), B=Get bIsFalling())
   Set Jumping(execute=@n2.then_2, Jumping:double=SelectFloat(A=1, B=0, bPickA=AND Boolean(A=@n7, B=@n6))) -> (Output_Get:double)
   n8 = Get bIsSliding()
   n9 = Get bIsSlopeSliding() -> (bIsSlopeSliding:bool)
   Set Slide(execute=@n2.then_3, Slide:double=SelectFloat(A=1, B=0, bPickA=OR Boolean!(A=@n9, B=@n8) -> (ReturnValue))) -> (Output_Get:double)
   > n10 = Branch(Condition="float > float"(A=Get PawnSpeed2D(), B=1600) {is=title K2Node_CallFunction})
   > Set zipline(zipline:double=1) -> (Output_Get:double)
   Set zipline(execute=@n10.else, zipline=0)
   n11 = PlayerMovement(execute=@n2.then_4, AnimInstance=Self-Reference(), XDotProductCoefficient=100, YDotProductCoefficient=100, ZDotProductCoefficient=100)
   n12 = BreakVector(InVec=MakeVector(X=@n11, Y=@n11.YMovement))
   > Set Left(Left:double="To Float (Boolean)"(InBool="float < float"(A=@n12, B=-50))) -> (Output_Get:double)
   > Set Right(Right:double="To Float (Boolean)"(InBool="float > float"(A=@n12, B=50) {is=title K2Node_PromotableOperator})) -> (Output_Get:double)
   n13 = "float > float"(A=@n12.Y, B=40)
   > Set Forward(Forward:double="To Float (Boolean)"(InBool=@n13)) -> (Output_Get:double)
   > Set Backward(Backward:double="To Float (Boolean)"(InBool="float < float"(A=@n12.Y, B=-40))) -> (Output_Get:double)
   n14 = Get bIsJumping()
   n15 = Get bIsFalling()
   > Set Jump_Forward(Jump_Forward:bool=AND Boolean(A=OR Boolean!(A=@n14, B=@n15, C:bool) -> (ReturnValue), B=@n13)) -> (Output_Get:bool)
  ]}
]
variables [
 Backward {type=number}
 Backward movement {type=boolean}
 Crouch_move {type=boolean}
 Crouching {type=number}
 Delta Time X {type=number}
 Forward {type=number}
 Forward movement {type=boolean}
 Gravity Override {type=Vector, default=<0 0 -980>}
 Gravity Override Params {type=GravityOverrideParamsStruct}
 Jump_Forward {type=boolean}
 Jumping {type=number}
 L Side Movement {type=boolean}
 Left {type=number}
 Pawn {type=FortPlayerPawn}
 R Side Movement {type=boolean}
 Right {type=number}
 SkelMesh {type=SkeletalMesh, default=null}
 Slide {type=number}
 Tactical_Jump {type=number}
 zipline {type=number}
]
